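# Editor utility widget for tech artists that counts how many actors of a chosen class, BuildingProp by default, are in the level.
# On Construct, two Details Views bind to the widget itself, filtered by category so one serves as config and the other as results.
# The "Collect All Listed" button gets all actors of the selected class and stores the count in NumberInLevel.
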EditorUtilityWidgetBlueprint /Game/ArtTools/Tech_Art/ActorInstanceChecker/EUW_ActorInstanceChecker
graphs [EventGraph {nodes [
   "On Clicked (TA_Generic01_Button)"!() -> (OutputDelegate, then, widget:User Widget Object Reference, event:Pointer Event Structure) {is=title K2Node_ComponentBoundEvent}
   > n1 = GetAllActorsOfClass!(
    execute
    WorldContextObject:Object Reference
    ActorClass:Actor Class Reference=Get ActorClassToCheck() -> (ActorClassToCheck:Actor Class Reference)
   ) -> (then, OutActors:Array of Actor Object References)
   > Set NumberInLevel(NumberInLevel:int=Length(TargetArray=@n1)) -> (Output_Get:int=0)
   Event Construct()
   > SetObject(
    self=[Get Config() -> (Config:Details View Object Reference), Get Results() -> (Results:Details View Object Reference)]
    NewObject=Self-Reference()
   )
   Event PreConstruct()
  ]}]
variables [
 ActorClassToCheck {type=Class, default=&BuildingProp}
 NumberInLevel {type=integer}
]
widgets {
 info {rootWidgetClass=&Overlay, widgetCount=7}
 widgets [
  Overlay_29 {widgetClassPath=&Overlay, widget=Overlay_29}
  HorizontalBox_37 {
   widgetClassPath=&HorizontalBox
   parent=Overlay_29
   widget=HorizontalBox_37
   slot {class=OverlaySlot, properties {horizontalAlignment=HAlign_Fill, verticalAlignment=VAlign_Fill}}
  }
  VerticalBox_73 {widgetClassPath=&VerticalBox, parent=HorizontalBox_37, widget=VerticalBox_73, slot {class=HorizontalBoxSlot}}
  TA_Generic01_Button {
   widgetClassPath=&TA_Generic01_Button
   parent=VerticalBox_73
   widget=TA_Generic01_Button
   properties {label=Collect All Listed}
   slot {class=VerticalBoxSlot}
  }
  Config {
   widgetClassPath=&DetailsView
   parent=VerticalBox_73
   widget=Config
   properties {categoriesToShow [NoCategory]}
   slot {class=VerticalBoxSlot}
  }
  VerticalBox_136 {widgetClassPath=&VerticalBox, parent=HorizontalBox_37, widget=VerticalBox_136, slot {class=HorizontalBoxSlot}}
  Results {
   widgetClassPath=&DetailsView
   parent=VerticalBox_136
   widget=Results
   properties {categoriesToShow [Results]}
   slot {class=VerticalBoxSlot}
  }
 ]
}
